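# Ambient spot sound cue for the Spawn Island point of interest, playing a stereo asteroid field bed with distant attenuation.
# A looping wave is mixed with a repeating, randomly delayed pick from six seismic rumbles borrowed from Artemis Hollow Base cave sounds.
# The quality switch gives its first two levels the full mix and the last two only the looping bed.
# A concurrency override limits simultaneous instances.
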SoundCue /Game/Sounds/Asteria/POI/SpawnIsland/Spots/Spot_Asteroid_Field_Loop_Stereo_Cue
root=QualityLevel_0
nodes [
 {id=Delay_0, children [Random_2], properties {delayMax=20, delayMin=12}}
 {id=Looping_0, children [Delay_0]}
 {id=Mixer_0, children [Looping_0,WavePlayer_0], properties {inputVolume [2,2]}}
 {id=QualityLevel_0, children [Mixer_0,Mixer_0,WavePlayer_0,WavePlayer_0]}
 {
  id=Random_2
  children [WavePlayer_20, WavePlayer_21, WavePlayer_22, WavePlayer_23, WavePlayer_24, WavePlayer_25]
  properties {weights [1,1,1,1,1,1]}
 }
 {id=WavePlayer_0, properties {bLooping=true, soundWaveAssetPtr=&/Spot_Asteroid_Field_Loop_Stereo}}
 {id=WavePlayer_20, properties {soundWaveAssetPtr=/Game/Sounds/Artemis/POI/Hollow_Base/Spots/Spot_Cave_Seismic_Rumble_01}}
 {id=WavePlayer_21, properties {soundWaveAssetPtr=/Game/Sounds/Artemis/POI/Hollow_Base/Spots/Spot_Cave_Seismic_Rumble_02}}
 {id=WavePlayer_22, properties {soundWaveAssetPtr=/Game/Sounds/Artemis/POI/Hollow_Base/Spots/Spot_Cave_Seismic_Rumble_03}}
 {id=WavePlayer_23, properties {soundWaveAssetPtr=/Game/Sounds/Artemis/POI/Hollow_Base/Spots/Spot_Cave_Seismic_Rumble_04}}
 {id=WavePlayer_24, properties {soundWaveAssetPtr=/Game/Sounds/Artemis/POI/Hollow_Base/Spots/Spot_Cave_Seismic_Rumble_05}}
 {id=WavePlayer_25, properties {soundWaveAssetPtr=/Game/Sounds/Artemis/POI/Hollow_Base/Spots/Spot_Cave_Seismic_Rumble_06}}
]
properties {
 attenuationSettings=&/Spot_Distant_Asteroid_Field_Stereo_Att
 bEnableBusSends=false
 bOverrideConcurrency=true
 maxDistance=45000
 soundClassObject=&Fort_Ambient_Outdoors_Athena
 volumeMultiplier=0.752
 concurrencyOverrides {maxCount=2}
 preEffectBusSends [{}]
}
