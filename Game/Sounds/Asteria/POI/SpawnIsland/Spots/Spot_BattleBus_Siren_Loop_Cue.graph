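# Looping Battle Bus siren ambience cue for a Spawn Island spot, in the outdoor ambient sound class and limited to one concurrent instance.
# An Enveloper fades volume in over two seconds above a Quality Level node whose first two slots share one mixer.
# That mixer blends a close-attenuated looping stereo siren with a distant-attenuation branch.
# The distant branch's sub-mixer has no inputs connected, so only the close wave supplies audio.
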SoundCue /Game/Sounds/Asteria/POI/SpawnIsland/Spots/Spot_BattleBus_Siren_Loop_Cue
root=Enveloper_0
nodes [
 {
  id=Attenuation_0
  children [WavePlayer_0]
  properties {attenuationSettings=&/Spot_BattleBus_Siren_Loop_Close}
 }
 {
  id=Attenuation_2
  children [Mixer_2]
  properties {attenuationSettings=&/Spot_BattleBus_Siren_Loop_Distant_Att}
 }
 {
  id=Enveloper_0
  children [QualityLevel_0]
  properties {pitchCurve {editorCurveData {keys [{time=0}]}}, volumeCurve {editorCurveData {keys [<at 0 0>,<at 2 1>]}}}
 }
 {id=Mixer_1, children [Attenuation_0, Attenuation_2], properties {inputVolume [1.25,1]}}
 {id=Mixer_2, children [null,null], properties {inputVolume [0.5,1]}}
 {id=QualityLevel_0, children [Mixer_1,Mixer_1,null,null]}
 {
  id=WavePlayer_0
  properties {bLooping=true, soundWaveAssetPtr=&/Spot_BattleBus_Siren_Close_Loop_Stereo}
 }
]
properties {
 bOverrideConcurrency=true
 maxDistance=6000
 soundClassObject=&Fort_Ambient_Outdoors_Athena
 volumeMultiplier=0.501
 attenuationOverrides {distanceAlgorithm=Custom, falloffDistance=10112.3027, attenuationShapeExtents {x=600}}
 concurrencyOverrides {maxCount=1}
}
