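# Sound cue for ominous space swell ambient one-shots at the Asteria SpawnIsland POI, using a wind one-shot attenuation.
# A quality-level root sends its first two slots to a weighted random pick that favors large swells two-to-one over medium ones.
# Its last two slots play random desert sand-gust waves borrowed from Artemis instead of swells.
# Modulators vary volume on each branch, and pitch on medium swells and gusts.
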SoundCue /Game/Sounds/Asteria/POI/SpawnIsland/Spots/Spot_Ominious_Space_Swells_Cue
root=QualityLevel_0
nodes [
 {id=Modulator_0, children [Random_4], properties {volumeMax=2.1, volumeMin=2}}
 {id=Modulator_1, children [Random_5], properties {pitchMax=1.1, volumeMax=1.2, volumeMin=1}}
 {id=Modulator_2, children [Random_1], properties {pitchMax=1.1, volumeMax=2.2, volumeMin=2}}
 {id=QualityLevel_0, children [Random_3,Random_3,Modulator_2,Modulator_2]}
 {id=Random_1, children [WavePlayer_5,WavePlayer_6,WavePlayer_7], properties {weights [1,1,1]}}
 {id=Random_3, children [Modulator_0,Modulator_1], properties {weights [1,0.5]}}
 {id=Random_4, children [WavePlayer_18, WavePlayer_19, WavePlayer_20, WavePlayer_21, WavePlayer_22], properties {weights [1,1,1,1,1]}}
 {id=Random_5, children [WavePlayer_23, WavePlayer_24, WavePlayer_25, WavePlayer_26, WavePlayer_27], properties {weights [1,1,1,1,1]}}
 {id=WavePlayer_18, properties {soundWaveAssetPtr=&/Spot_Ominious_Space_Swell_Large_01}}
 {id=WavePlayer_19, properties {soundWaveAssetPtr=&/Spot_Ominious_Space_Swell_Large_02}}
 {id=WavePlayer_20, properties {soundWaveAssetPtr=&/Spot_Ominious_Space_Swell_Large_03}}
 {id=WavePlayer_21, properties {soundWaveAssetPtr=&/Spot_Ominious_Space_Swell_Large_04}}
 {id=WavePlayer_22, properties {soundWaveAssetPtr=&/Spot_Ominious_Space_Swell_Large_05}}
 {id=WavePlayer_23, properties {soundWaveAssetPtr=&/Spot_Ominious_Space_Swell_Medium_01}}
 {id=WavePlayer_24, properties {soundWaveAssetPtr=&/Spot_Ominious_Space_Swell_Medium_02}}
 {id=WavePlayer_25, properties {soundWaveAssetPtr=&/Spot_Ominious_Space_Swell_Medium_03}}
 {id=WavePlayer_26, properties {soundWaveAssetPtr=&/Spot_Ominious_Space_Swell_Medium_04}}
 {id=WavePlayer_27, properties {soundWaveAssetPtr=&/Spot_Ominious_Space_Swell_Medium_05}}
 {id=WavePlayer_5, properties {soundWaveAssetPtr=/Game/Sounds/Artemis/Biomes/Desert/Desert_Spots/Spot_Desert_Wind_Sand_Gust_02}}
 {id=WavePlayer_6, properties {soundWaveAssetPtr=/Game/Sounds/Artemis/Biomes/Desert/Desert_Spots/Spot_Desert_Wind_Sand_Gust_01}}
 {id=WavePlayer_7, properties {soundWaveAssetPtr=/Game/Sounds/Artemis/Biomes/Desert/Desert_Spots/Spot_Desert_Wind_Sand_Gust_05}}
]
properties {
 attenuationSettings=/Game/Sounds/Attenuation/Fort_TileAmbient_OneShot_Wind_Att
 bEnableBusSends=false
 maxDistance=8000
 soundClassObject=&Fort_Ambient_OneShots
 volumeMultiplier=0.501
 preEffectBusSends [{}]
}
